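# Masked, two-sided lit material for Niagara mesh and ribbon particles in the Clash V pickaxe effects.
# Dynamic parameters drive a sine wobble with normal extension for vertex displacement, and a sphere-mask dissolve broken up by panning noise.
# Fresnel blends base colors tinted by particle color, over a panning swirl normal map.
# Emissive blends by sun-facing normals with time-of-day tint, and temporal dither scales a depth threshold.
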Material /BRCosmetics/Weapons/FORT_Melee/Pickaxe_Clash_V_Male_1h/FX/M_Pickaxe_Clash_V_Mesh
graph {
 properties {
  bUsedWithNiagaraMeshParticles=true
  bUsedWithNiagaraRibbons=true
  blendMode=BLEND_Masked
  pixelDepthOffsetMode=PDOM_Legacy
  twoSided=true
  usedShadingModels=Default Lit
 }
 nodes [
  n1 = VertexNormalWS()
  n2 = DynamicParameter() {parameterIndex=1, defaultValue=<rgba 1 2 1 1>, paramNames [WPO_Intensity, WPO_Speed, WPO_Extend, ""]}
  n3 = ComponentMask(None=TextureCoordinate() {coordinateIndex=1, uTiling=0.4, vTiling=0.5}) {g=true}
  n4 = Constant() {r=2}
  n5 = Time() {period=2}
  n6 = Sine(None=Add(A=Multiply(A=@n3, B=@n4), B=Multiply(A=@n5, B=@n2.WPO_Speed) {constB=2}))
  n7 = TextureCoordinate() {coordinateIndex=1}
  n8 = Constant() {r=0.5}
  n9 = Multiply(
   A=MaterialFunctionCall(X=@n6, Y=@n6, Z=@n6) {materialFunction=&MakeFloat3}.Result
   B=SphereMask(A=ComponentMask(None=@n7) {g=true}, B=@n8, Radius=@n8) {hardnessPercent=0}
  )
  Add(
   A=Multiply(A=Normalize(VectorInput=@n1), B=@n2.WPO_Extend)
   B=Multiply(A=@n2.WPO_Intensity, B=Transform(None=@n9) {transformSourceType=TRANSFORMSOURCE_Local, transformType=TRANSFORM_Tangent})
  )
  n10 = DynamicParameter() {paramNames [Dissolve_Mix, Dissolve_Process, "", ""]}
  n11 = SphereMask(A=ComponentMask(None=TextureCoordinate() {coordinateIndex=1}) {g=true}, B=@n10.Dissolve_Process, Radius=Constant() {r=2}) {attenuationRadius=1, hardnessPercent=0}
  n12 = TextureCoordinate() {coordinateIndex=1}
  n13 = VectorParameter() {
   group=Dissolve
   parameterName=Dissolve UV
   channelNames {a=Dissolve Pan Speed V, b=Dissolvel Pan Speed U, g=Dissolve Tile V, r=Dissolve Tile U}
   defaultValue=<rgba 0.5 0.3 0 0.2>
  }
  n14 = Multiply(A=ComponentMask(None=@n12) {r=true}, B=@n13.Dissolve Tile U)
  n15 = Multiply(A=ComponentMask(None=@n12) {g=true}, B=@n13.Dissolve Tile V)
  n16 = TextureSampleParameter2D(UVs=Panner(Coordinate=AppendVector(A=@n14, B=@n15), Speed=AppendVector(A=@n13.Dissolvel Pan Speed U, B=@n13.Dissolvel Pan Speed U)) {speedX=-1, speedY=0.1}) {
   group=Dissolve
   parameterName=Dissolve Tex
   samplerType=SAMPLERTYPE_Grayscale
   texture=&LowResBlurredNoise
  }
  LinearInterpolate(
   A=Constant() {r=1}
   B=LinearInterpolate(A=@n11, B=Multiply(A=@n11, B=@n16.R), Alpha=@n10.Dissolve_Mix) {constAlpha=0.7}
   Alpha=ScalarParameter() {defaultValue=1, parameterName=Ues Dissolve}
  )
  n17 = MaterialFunctionCall(
   Color_Morning=VectorParameter() {group=TOD, parameterName=TOD_Morning, defaultValue=<rgba 0.3675 0.1266 0.0895 1>}.RGB
   Color_Day=VectorParameter() {group=TOD, parameterName=TOD_Afternoon, defaultValue=<rgba 1 1 1 1>}.RGB
   Color_Evening=VectorParameter() {group=TOD, parameterName=TOD_Evening, defaultValue=<rgba 0.3316 0.1538 0.1088 1>}.RGB
   Color_Night=VectorParameter() {group=TOD, parameterName=TOD_Night, defaultValue=<rgba 0.0874 0.1753 0.1753 1>}.RGB
  ) {materialFunction=&Color_Time_of_Day}
  n18 = CollectionParameter() {
   collection=&FortniteMaterialParameters
   parameterName=SunAndMoonAtmosphereDirectionalVector
  }
  n19 = ComponentMask(None=@n18) {b=true, g=true, r=true}
  n20 = PixelNormalWS()
  n21 = LinearInterpolate(
   A=VectorParameter() {parameterName=emissiveColorAgainstSun, defaultValue=<rgba 0.0091 0.0098 0.013 0>}.RGB
   B=VectorParameter() {parameterName=EmissiveColorInSun}.RGB
   Alpha=Clamp(None=DotProduct(A=@n19, B=@n20))
  )
  MaterialFunctionCall(Emissive=Multiply(A=Multiply(A=@n17.Result, B=@n21), B=ScalarParameter() {parameterName=EmissiveIntensity})) {materialFunction=&MF_FX_EmissiveAdjustments}
  n22 = TextureCoordinate()
  n23 = AppendVector(
   A=Multiply(A=ComponentMask(None=@n22) {r=true}, B=ScalarParameter() {defaultValue=0.7, group=Normal, parameterName=Normal Tile U})
   B=Multiply(A=ComponentMask(None=@n22) {g=true}, B=ScalarParameter() {defaultValue=8, group=Normal, parameterName=Normal Tile V})
  )
  n24 = Time() {bOverride_Period=true, period=64}
  n25 = AppendVector(
   A=ScalarParameter() {group=Normal, parameterName=Normal Pan Speed U}
   B=ScalarParameter() {defaultValue=1, group=Normal, parameterName=Normal Pan Speed V}
  )
  Multiply(
   A=TextureSample(UVs=Panner(Coordinate=@n23, Time=@n24, Speed=@n25) {speedX=-1, speedY=0.1}) {samplerType=SAMPLERTYPE_Normal, texture=/Game/Global/Textures/Default/Noise/T_SwirlNoise_N}.RGB
   B=VectorParameter() {group=Normal, parameterName=NormalStrenght, defaultValue=<rgba 0.5 0.5 1 1>}.RGB
  )
  n26 = ScalarParameter() {defaultValue=1, parameterName=FresnelMult}
  n27 = ScalarParameter() {defaultValue=0.5, parameterName=FresnelExp}
  Multiply(
   A=LinearInterpolate(
    A=VectorParameter() {parameterName=baseColorOuter, defaultValue=<rgba 0.007 0.0079 0.01 0>}.RGB
    B=VectorParameter() {parameterName=baseColorInner, defaultValue=<rgba 0.021 0.0227 0.03 0>}.RGB
    Alpha=MaterialFunctionCall(Brightness=@n26, Exponent=@n27) {materialFunction=&MF_Fresnel_FX}.Fresnel
   )
   B=ParticleColor().RGB
  )
  Multiply(
   A=ScalarParameter() {defaultValue=5, parameterName=DitherDepthThershold}
   B=MaterialFunctionCall() {materialFunction=&DitherTemporalAA}.Result
  )
  ScalarParameter() {defaultValue=1, parameterName=Specular}
  ScalarParameter() {defaultValue=0.16, parameterName=Roughness}
  ScalarParameter() {defaultValue=0.2, parameterName=Metalic}
 ]
}
