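# Idle Niagara effect for the Clash V male one-handed pickaxe, combining burst meshes with two ribbon layers.
# Mesh_Burst spawns two slowly rotating meshes that follow the lowerarm_r bone, with curve-driven scale and material parameters.
# Ribbon_BoneFlow uses the FN_BoneFlow module to shape ribbons along three BoneFlow sockets, facing them along the spline binormal.
# Ribbon_Tendrils emits tapering ribbons from User.SkeletalMesh, pushed by cone velocity and panning curl noise.
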NiagaraSystem /BRCosmetics/Weapons/FORT_Melee/Pickaxe_Clash_V_Male_1h/FX/NS_Pickaxe_Clash_V_Idle
dependencies {usedDataInterfaces [&NiagaraDataInterfaceCurve], usedDynamicInputs [], usedModules [&SystemState]}
emitters {
 Mesh_Burst {
  emitterUpdateScript [
   &EmitterState(Inactive Response=::NewEnumerator1, Life Cycle Mode=::NewEnumerator1, Loop Behavior=::NewEnumerator1, Loop Duration=4, UseLoopDelay=false)
   &SpawnBurst_Instantaneous(Spawn Count=2, Use Spawn Probability=false, UseLoopCountLimit=false)
  ]
  particleSpawnScript [
   &InitializeParticle(
    Lifetime Max=1.6
    Lifetime Min=2
    Lifetime Mode=::NewEnumerator1
    Mesh Scale=&RandomRangeVector(Minimum=<1.2 0.6 0.8>, Maximum=<1.4 0.5 0.7>)
    Mesh Scale Mode=::NewEnumerator4
    UsePositionOffset=false
    Write Material Random=true
   )
   &InitialMeshOrientation(
    Rotation=&RandomRangeVector(Minimum=<-0.4 0 0>, Maximum=<0.2 0 0>)
    Rotation Coordinate Space=::NewEnumerator2
    Use Orientation Vector=true
    Use Rotation Vector=true
   )
  ]
  particleUpdateScript [
   &ParticleState()
   &UpdateMeshOrientation(Rotation Rate=-0.1, Rotation Vector=<1 0 0>)
   &SkeletalMeshLocation(
    Bone Sampling Mode=::NewEnumerator4
    MaskSpawnGroup=false
    Orientation Sampling=::NewEnumerator1
    Sampled Position Offset=&RandomRangeVector(Minimum=<-20 0 0.2>, Maximum=<-30 1 0.2>)
    Sampled Position Offset Coordinate Space:ENiagaraSkelMeshTransforms=::NewEnumerator2
    Skeletal Mesh=DataInterface() {propertyValues {
      DefaultMesh=null
      ExcludeBoneName=root
      OptionalComponentClass=null
      PreviewMesh=/BRCosmetics/Weapons/FORT_Melee/Pickaxe_Clash_V_Male_1h/Meshes/Clash_V_Male_1h_Axe
      SkinningMode=SkinOnTheFly
      SoftSourceActor=null
      SourceMode=Default
      UvSetIndex=0
      WholeMeshLOD=-1
      bExcludeBone=false
      bReadDeformedGeometry=true
      bRequireCurrentFrameData=true
      ComponentTags []
      FilteredBones [lowerarm_r]
      FilteredSockets []
      MeshUserParameter {parameter {name=None, typeDefHandle {registeredTypeIndex=1}, varData []}}
      SamplingRegions []
     }}
   )
   &DynamicMaterialParameters(
    Index 0 Param 1=&FloatFromCurve(FloatCurve=DataInterface() {propertyValues {Curve {keys [<at 0 1>,<at 0.5 0.5>]}}})
    Index 0 Param 2=&FloatFromCurve(FloatCurve=DataInterface() {propertyValues {Curve {keys [
         {interpMode=RCIM_Cubic, leaveTangent=-2.9591, tangentMode=RCTM_Break, time=0, value=1.5}
         {arriveTangent=-0.3114, interpMode=RCIM_Cubic, tangentMode=RCTM_Break, time=1, value=0}
        ]}}})
    Index 1 Param 1=&FloatFromCurve(FloatCurve=DataInterface() {propertyValues {Curve {keys [
         {arriveTangent=-10.7468, interpMode=RCIM_Cubic, leaveTangent=-10.7468, tangentMode=RCTM_User, time=0, value=2}
         {interpMode=RCIM_Cubic, leaveTangent=0.5305, tangentMode=RCTM_Break, time=0.1, value=1}
         {interpMode=RCIM_Cubic, tangentMode=RCTM_Break, time=1, value=0.2}
        ]}}})
    Index 1 Param 2=2
    Index 1 Param 3=&FloatFromCurve(FloatCurve=DataInterface() {propertyValues {Curve {keys [
         {interpMode=RCIM_Cubic, leaveTangent=10.6169, tangentMode=RCTM_Break, time=0, value=-0.2}
         {arriveTangent=2.4889, interpMode=RCIM_Cubic, leaveTangent=0.1906, tangentMode=RCTM_Break, time=0.1, value=0.5}
         {arriveTangent=-0.3416, interpMode=RCIM_Cubic, leaveTangent=-1.8967, tangentMode=RCTM_Break, time=0.6, value=0.5}
         {arriveTangent=0.3626, interpMode=RCIM_Cubic, tangentMode=RCTM_Break, time=0.9, value=0}
        ]}}})
    Module.Param0WriteEnabled=true
    Module.Param0WriteEnabled1=true
    Module.Param1WriteEnabled=true
    Module.Param1WriteEnabled1=true
    Module.Param2WriteEnabled=false
    Module.Param2WriteEnabled1=true
    Module.Param3WriteEnabled=false
    Module.Param3WriteEnabled1=false
    Write Parameter Index 1=true
   )
   &ScaleMeshSize(Scale Factor=&MakeVector(
     X=&FloatFromCurve(FloatCurve=DataInterface() {propertyValues {Curve {keys [<at 0.65 1>,<at 1 1.2>]}}})
     Y=&FloatFromCurve(FloatCurve=DataInterface() {propertyValues {Curve {keys [
          {interpMode=RCIM_Cubic, tangentMode=RCTM_User, time=0, value=0.7}
          {interpMode=RCIM_Cubic, tangentMode=RCTM_User, time=0.6, value=1}
          {arriveTangent=-4.4791, interpMode=RCIM_Cubic, leaveTangent=-4.4791, tangentMode=RCTM_User, time=1, value=0}
         ]}}})
     Z=&FloatFromCurve(FloatCurve=DataInterface() {propertyValues {Curve {keys [
          {interpMode=RCIM_Cubic, tangentMode=RCTM_User, time=0, value=0.7}
          {interpMode=RCIM_Cubic, tangentMode=RCTM_User, time=0.6, value=1}
          {arriveTangent=-6.148, interpMode=RCIM_Cubic, leaveTangent=-6.148, tangentMode=RCTM_User, time=1, value=0}
         ]}}})
    ))
  ]
  rendererClasses [&NiagaraMeshRendererProperties]
  renderers [{rendererClass=&NiagaraMeshRendererProperties, rendererIndex=0}]
 }
 Ribbon_BoneFlow {
  emitterUpdateScript [
   &EmitterState(Life Cycle Mode=::NewEnumerator1, Loop Duration=1, UseLoopDelay=false)
   &SpawnRate(SpawnRate=&RandomRangeFloat(Minimum=20, Maximum=30), Use Spawn Probability=false)
  ]
  particleSpawnScript [
   &FN_SetSpawnGroup(Inclusivity=true, Normalize Spawn Group=true, Particles Per Group=10, Pattern=::NewEnumerator0, Set Ribbon ID and Ribbon Link Order=true)
   &InitializeRibbon(
    Lifetime Max=1.5
    Lifetime Min=0.8
    Lifetime Mode:ENiagara_LifetimeMode=::NewEnumerator1
    Ribbon Width=&RandomRangeFloat(Minimum=2, Maximum=1.5)
    UsePositionOffset=false
    Write Material Random=true
   )
  ]
  particleUpdateScript [
   &ParticleState()
   /Game/Effects/Niagara/Modules/BoneFlow/FN_BoneFlow_v4/FN_BoneFlow(
    Angle:NiagaraFloat=&Multiply_Float(A=$Particles.Normalized Spawn Group, B=&RandomRangeFloat(Minimum=-0.6, Maximum=0.6))
    BoneID:NiagaraInt32=&ModuloInt(A=$Particles.SpawnGroup, B=1)
    Height:NiagaraFloat=&Multiply_Float(
     A=$Particles.Normalized Spawn Group
     B=&FloatFromCurve(FloatCurve=DataInterface() {propertyValues {Curve {keys [
          {arriveTangent=0.2329, interpMode=RCIM_Cubic, leaveTangent=0.2329, tangentMode=RCTM_User, time=0, value=1}
          {interpMode=RCIM_Cubic, tangentMode=RCTM_User, time=1, value=1.1}
         ]}}})
    )
    Radius:NiagaraFloat=&Multiply_Float(
     A=$Particles.Normalized Spawn Group
     B=&FloatFromCurve(
      FloatCurve=DataInterface() {propertyValues {Curve {keys [
          {arriveTangent=4.8142, interpMode=RCIM_Cubic, leaveTangent=4.8142, tangentMode=RCTM_User, time=0, value=10}
          {arriveTangent=-26.8012, interpMode=RCIM_Cubic, leaveTangent=-26.8012, tangentMode=RCTM_User, time=0.4, value=10}
          {arriveTangent=-23.2104, interpMode=RCIM_Cubic, leaveTangent=-23.2104, tangentMode=RCTM_User, time=1, value=0}
         ]}}}
      CurveIndex=$Particles.Normalized Spawn Group
      Scale Curve=&RandomRangeFloat(Minimum=2, Maximum=1)
     )
    )
    SkeletalMesh=DataInterface() {propertyValues {
      DefaultMesh=null
      ExcludeBoneName=root
      OptionalComponentClass=null
      PreviewMesh=&Clash_V_Male_1h_Axe
      SkinningMode=SkinOnTheFly
      SoftSourceActor=null
      SourceMode=Default
      UvSetIndex=0
      WholeMeshLOD=-1
      bExcludeBone=false
      bReadDeformedGeometry=true
      bRequireCurrentFrameData=true
      ComponentTags []
      FilteredBones []
      FilteredSockets [BoneFlow_00,BoneFlow_01,BoneFlow_02]
      MeshUserParameter {parameter {name=None, typeDefHandle {registeredTypeIndex=1}, varData []}}
      SamplingRegions []
     }}
    Use Sockets Instead=true
    Write Position Attribute=true
   ) {module=FN_BoneFlow001}
   &ScaleRibbonWidth(Ribbon Width Scale=&FloatFromCurve(
     FloatCurve=DataInterface() {propertyValues {Curve {keys [
         {interpMode=RCIM_Cubic, tangentMode=RCTM_User, time=0, value=0}
         {interpMode=RCIM_Cubic, tangentMode=RCTM_User, time=0.2, value=2}
         {arriveTangent=-8.5417, interpMode=RCIM_Cubic, leaveTangent=-8.5417, tangentMode=RCTM_User, time=0.6, value=0.575}
         {arriveTangent=-2.6171, interpMode=RCIM_Cubic, leaveTangent=-2.6171, tangentMode=RCTM_User, time=0.8, value=1}
         {arriveTangent=-0.2025, interpMode=RCIM_Cubic, leaveTangent=-0.2025, tangentMode=RCTM_User, time=1, value=0}
        ]}}}
     CurveIndex=$Particles.Normalized Spawn Group
     Scale Curve=&FloatFromCurve(
      FloatCurve=DataInterface() {propertyValues {Curve {keys [
          {arriveTangent=2.8989, interpMode=RCIM_Cubic, leaveTangent=2.8989, tangentMode=RCTM_User, time=0.01, value=0}
          {interpMode=RCIM_Cubic, tangentMode=RCTM_User, time=0.4, value=0.95}
          {interpMode=RCIM_Cubic, tangentMode=RCTM_User, time=0.6, value=1}
          {arriveTangent=-1.5059, interpMode=RCIM_Cubic, leaveTangent=-1.5059, tangentMode=RCTM_User, time=0.99, value=0}
         ]}}}
      CurveIndex=$Particles.Age
     )
    ))
   /SetVariables(Particles.RibbonFacing=$Output.FN_BoneFlow001.SplineBinormal) {isSetParameters=true}
   &DynamicMaterialParameters(
    Index 0 Param 1=&FloatFromCurve(FloatCurve=DataInterface() {propertyValues {Curve {keys [<at 0 0 RCIM_Cubic>,<at 0.5 1>,<at 1 0 RCIM_Cubic>]}}})
    Module.Param0WriteEnabled=true
    Module.Param1WriteEnabled=true
    Module.Param2WriteEnabled=false
    Module.Param3WriteEnabled=false
   )
   &SolveForcesAndVelocity(Clamp Velocity=false, Limit Acceleration=false)
  ]
  rendererClasses [&NiagaraRibbonRendererProperties]
  renderers [{rendererClass=&NiagaraRibbonRendererProperties}]
 }
 Ribbon_Tendrils {
  emitterUpdateScript [
   &EmitterState(Life Cycle Mode=::NewEnumerator1, Loop Duration=1, UseLoopDelay=false)
   &FN_SpawnRibbon_v2(Number of Ribbons=2, SpawnRate=2) {module=FN_SpawnRibbon_v2001}
  ]
  particleSpawnScript [
   /Game/Effects/Niagara/Modules/Ribbons/FN_InitializeRibbon_v4(
    Color:LinearColor=<rgba 0.276 0.276 0.276 1>
    Custom Offset=true
    Enable Skeletal Position=true
    Lifetime:NiagaraFloat=&RandomRangeFloat(Minimum=2, Maximum=2.5)
    Local Vector Offset:Vector3f=<1 0.2 0.2>
    Offset Amount:NiagaraFloat=&RandomRangeFloat(Minimum=-15, Maximum=-15)
    Ribbon Width:NiagaraFloat=&RandomRangeFloat(Minimum=8, Maximum=4)
    SkeletalMesh:NiagaraDataInterfaceSkeletalMesh=$User.SkeletalMesh
   )
   &AddVelocityInCone(Cone Angle=40, Cone Axis=<-1 0.3 -0.05>, Module.Use Velocity Falloff On Cone Axis=false, Velocity Strength=20)
  ]
  particleUpdateScript [
   /Niagara/Modules/Update/Forces/CurlNoiseForce(
    Noise Frequency=0.03
    Pan Noise=true
    Pan Noise Field=<0.05 0.05 0.05>
    Random Seed:NiagaraInt32=&RandomRangeInt(Minimum=0, Maximum=1)
   )
   &SolveForcesAndVelocity(Clamp Velocity=false, Limit Acceleration=false)
   /Game/Effects/Niagara/Modules/Ribbons/FN_UpdateRibbon_v4(Enable Skeletal Position=true, SkeletalMesh:NiagaraDataInterfaceSkeletalMesh=$User.SkeletalMesh)
   &DynamicMaterialParameters(
    Index 0 Param 1=&FloatFromCurve(FloatCurve=DataInterface() {propertyValues {Curve {keys [<at 0 1>,<at 1 -0.5>]}}})
    Index 0 Param 2=$Particles.MaterialRandom
    Index 0 Param 3=1
    Module.Param0WriteEnabled=true
    Module.Param1WriteEnabled=true
    Module.Param2WriteEnabled=false
    Module.Param3WriteEnabled=true
   ) {module=DynamicMaterialParameters001}
   &ScaleRibbonWidth(Ribbon Width Scale=&FloatFromCurve(FloatCurve=DataInterface() {propertyValues {Curve {keys [
         <at 0 1 RCIM_Cubic>
         <at 0.3 1>
         {arriveTangent=-1.6026, interpMode=RCIM_Cubic, leaveTangent=-1.6026, tangentMode=RCTM_User, time=0.6, value=0.275}
         {arriveTangent=-2.0105, interpMode=RCIM_Cubic, leaveTangent=-2.0105, tangentMode=RCTM_User, time=1, value=-0.5}
        ]}}}))
  ]
  rendererClasses [&NiagaraRibbonRendererProperties]
  renderers [{rendererClass=&NiagaraRibbonRendererProperties}]
 }
}
summary {
 systemName=NS_Pickaxe_Clash_V_Idle
 emitters [
  {emitterName=Mesh_Burst, rendererClasses [&NiagaraMeshRendererProperties]}
  {emitterName=Ribbon_BoneFlow, rendererClasses [&NiagaraRibbonRendererProperties]}
  {emitterName=Ribbon_Tendrils, rendererClasses [&NiagaraRibbonRendererProperties]}
 ]
 userVariables [
  User.RibbonIDOffset {type=NiagaraInt32}
  User.SkeletalMesh {
   type=NiagaraDataInterfaceSkeletalMesh
   defaultValue {
    struct=NiagaraExt_VariableValue_DataInterface
    value {dataInterface=&self:NiagaraDataInterfaceSkeletalMesh_1}
   }
  }
 ]
}
