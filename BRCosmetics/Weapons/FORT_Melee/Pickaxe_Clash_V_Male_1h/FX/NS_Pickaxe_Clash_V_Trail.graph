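# Niagara ribbon trail effect for the one-handed Clash V pickaxe, built on AnimTrail modules that spawn ribbons between indexed bone or socket pairs.
# Four short-lived grey tendril emitters use different widths and lifetimes, with panning curl noise, fixed ribbon facing and curve-driven width and material parameters.
# A Trail_Blade emitter adds a smoother, longer-lived ribbon, colored by Fortnite time of day, that tapers and fades over its life.
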NiagaraSystem /BRCosmetics/Weapons/FORT_Melee/Pickaxe_Clash_V_Male_1h/FX/NS_Pickaxe_Clash_V_Trail
dependencies {
 usedDataInterfaces [&NiagaraDataInterfaceCurve, &NiagaraDataInterfaceColorCurve]
 usedDynamicInputs []
 usedModules [&SystemState]
}
emitters {
 Tendril_Trail {
  emitterUpdateScript [
   &EmitterState(Life Cycle Mode=::NewEnumerator1, Loop Behavior=::NewEnumerator1, Loop Duration=0.3, UseLoopDelay=false)
   &FN_SpawnAnimTrail_v2()
  ]
  particleSpawnScript [
   /Game/Effects/Niagara/Modules/AnimTrail/FN_InitializeAnimTrail/FN_InitializeAnimTrailBoneIndecies(First Bone or Socket Index=3, Second Bone or Socket Index=1)
   /Game/Effects/Niagara/Modules/AnimTrail/FN_InitializeAnimTrail/FN_InitializeAnimTrail_v2(AnimTrailWidth=0.8, Color:LinearColor=<rgba 0.276 0.276 0.276 1>, Lifetime=0.12)
  ]
  particleUpdateScript [
   &FN_UpdateAnimTrail_v2()
   /SetVariables(Particles.RibbonFacing=<1 0 0>) {isSetParameters=true, module=SetVariables_#1}
   /Niagara/Modules/Update/Forces/CurlNoiseForce(
    Noise Strength=2500
    Pan Noise=true
    Pan Noise Field=<0.1 0.1 0.1>
    Random Seed:NiagaraInt32=&RandomRangeInt(Minimum=0, Maximum=1)
   )
   &SolveForcesAndVelocity(Clamp Velocity=false, Limit Acceleration=false)
   &DynamicMaterialParameters(
    Index 0 Param 1=&FloatFromCurve(FloatCurve=DataInterface() {propertyValues {Curve {keys [<at 0 1 RCIM_Cubic>, {arriveTangent=-2, interpMode=RCIM_Cubic, leaveTangent=-2, time=0.665, value=0.6902}, <at 1 -1 RCIM_Cubic>]}}})
    Index 0 Param 2=$Particles.MaterialRandom
    Index 0 Param 3=0.6
    Index 0 Param 4=1
    Module.Param0WriteEnabled=true
    Module.Param1WriteEnabled=true
    Module.Param2WriteEnabled=true
    Module.Param3WriteEnabled=true
   )
   &ScaleRibbonWidth(Ribbon Width Scale=&FloatFromCurve(
     FloatCurve=DataInterface() {propertyValues {Curve {keys [
         <at 0 1 RCIM_Cubic>
         {arriveTangent=-1.6026, interpMode=RCIM_Cubic, leaveTangent=-1.6026, tangentMode=RCTM_User, time=0.7967, value=0.5043}
         {arriveTangent=-2.0105, interpMode=RCIM_Cubic, leaveTangent=-2.0105, tangentMode=RCTM_User, time=1, value=0}
        ]}}}
     Scale Curve=0.18
    ))
  ]
  rendererClasses [&NiagaraRibbonRendererProperties]
  renderers [{rendererClass=&NiagaraRibbonRendererProperties}]
 }
 Tendril_Trail001 {
  emitterUpdateScript [
   &EmitterState(Life Cycle Mode=::NewEnumerator1, Loop Behavior=::NewEnumerator1, Loop Duration=0.3, UseLoopDelay=false)
   &FN_SpawnAnimTrail_v2()
  ]
  particleSpawnScript [
   &FN_InitializeAnimTrailBoneIndecies(First Bone or Socket Index=2, Second Bone or Socket Index=3)
   &FN_InitializeAnimTrail_v2(AnimTrailWidth=1.3, Color:LinearColor=<rgba 0.276 0.276 0.276 1>, Lifetime=0.06)
  ]
  particleUpdateScript [
   &FN_UpdateAnimTrail_v2()
   /SetVariables(Particles.RibbonFacing=<1 0 0>) {isSetParameters=true, module=SetVariables_#1}
   &CurlNoiseForce(
    Noise Strength=2500
    Pan Noise=true
    Pan Noise Field=<0.1 0.1 0.1>
    Random Seed:NiagaraInt32=&RandomRangeInt(Minimum=0, Maximum=1)
   )
   &SolveForcesAndVelocity(Clamp Velocity=false, Limit Acceleration=false)
   &DynamicMaterialParameters(
    Index 0 Param 1=&FloatFromCurve(FloatCurve=DataInterface() {propertyValues {Curve {keys [<at 0 1 RCIM_Cubic>, {arriveTangent=-2, interpMode=RCIM_Cubic, leaveTangent=-2, time=0.665, value=0.6902}, <at 1 -1 RCIM_Cubic>]}}})
    Index 0 Param 2=$Particles.MaterialRandom
    Index 0 Param 3=0.6
    Index 0 Param 4=1
    Module.Param0WriteEnabled=true
    Module.Param1WriteEnabled=true
    Module.Param2WriteEnabled=true
    Module.Param3WriteEnabled=true
   )
   &ScaleRibbonWidth(Ribbon Width Scale=&FloatFromCurve(
     FloatCurve=DataInterface() {propertyValues {Curve {keys [
         <at 0 1 RCIM_Cubic>
         {arriveTangent=-1.6026, interpMode=RCIM_Cubic, leaveTangent=-1.6026, tangentMode=RCTM_User, time=0.7967, value=0.5043}
         {arriveTangent=-2.0105, interpMode=RCIM_Cubic, leaveTangent=-2.0105, tangentMode=RCTM_User, time=1, value=0}
        ]}}}
     Scale Curve=0.18
    ))
  ]
  rendererClasses [&NiagaraRibbonRendererProperties]
  renderers [{rendererClass=&NiagaraRibbonRendererProperties}]
 }
 Tendril_Trail002 {
  emitterUpdateScript [
   &EmitterState(Life Cycle Mode=::NewEnumerator1, Loop Behavior=::NewEnumerator1, Loop Duration=0.3, UseLoopDelay=false)
   &FN_SpawnAnimTrail_v2()
  ]
  particleSpawnScript [
   &FN_InitializeAnimTrailBoneIndecies(First Bone or Socket Index=4, Second Bone or Socket Index=5)
   &FN_InitializeAnimTrail_v2(AnimTrailWidth=1, Color:LinearColor=<rgba 0.276 0.276 0.276 1>, Lifetime=0.09)
  ]
  particleUpdateScript [
   &FN_UpdateAnimTrail_v2()
   /SetVariables(Particles.RibbonFacing=<1 0 0>) {isSetParameters=true, module=SetVariables_#1}
   &CurlNoiseForce(
    Noise Strength=2500
    Pan Noise=true
    Pan Noise Field=<0.1 0.1 0.1>
    Random Seed:NiagaraInt32=&RandomRangeInt(Minimum=0, Maximum=1)
   )
   &SolveForcesAndVelocity(Clamp Velocity=false, Limit Acceleration=false)
   &DynamicMaterialParameters(
    Index 0 Param 1=&FloatFromCurve(FloatCurve=DataInterface() {propertyValues {Curve {keys [<at 0 1 RCIM_Cubic>, {arriveTangent=-2, interpMode=RCIM_Cubic, leaveTangent=-2, time=0.665, value=0.6902}, <at 1 -1 RCIM_Cubic>]}}})
    Index 0 Param 2=$Particles.MaterialRandom
    Index 0 Param 3=0.6
    Index 0 Param 4=1
    Module.Param0WriteEnabled=true
    Module.Param1WriteEnabled=true
    Module.Param2WriteEnabled=true
    Module.Param3WriteEnabled=true
   )
   &ScaleRibbonWidth(Ribbon Width Scale=&FloatFromCurve(
     FloatCurve=DataInterface() {propertyValues {Curve {keys [
         <at 0 1 RCIM_Cubic>
         {arriveTangent=-1.6026, interpMode=RCIM_Cubic, leaveTangent=-1.6026, tangentMode=RCTM_User, time=0.7967, value=0.5043}
         {arriveTangent=-2.0105, interpMode=RCIM_Cubic, leaveTangent=-2.0105, tangentMode=RCTM_User, time=1, value=0}
        ]}}}
     Scale Curve=0.18
    ))
  ]
  rendererClasses [&NiagaraRibbonRendererProperties]
  renderers [{rendererClass=&NiagaraRibbonRendererProperties}]
 }
 Tendril_Trail003 {
  emitterUpdateScript [
   &EmitterState(Life Cycle Mode=::NewEnumerator1, Loop Behavior=::NewEnumerator1, Loop Duration=0.3, UseLoopDelay=false)
   &FN_SpawnAnimTrail_v2()
  ]
  particleSpawnScript [
   &FN_InitializeAnimTrailBoneIndecies(Second Bone or Socket Index=1)
   &FN_InitializeAnimTrail_v2(AnimTrailWidth=1.2, Color:LinearColor=<rgba 0.276 0.276 0.276 1>, Lifetime=0.06)
  ]
  particleUpdateScript [
   &FN_UpdateAnimTrail_v2()
   /SetVariables(Particles.RibbonFacing=<1 0 0>) {isSetParameters=true, module=SetVariables_#1}
   &CurlNoiseForce(
    Noise Strength=2500
    Pan Noise=true
    Pan Noise Field=<0.1 0.1 0.1>
    Random Seed:NiagaraInt32=&RandomRangeInt(Minimum=0, Maximum=1)
   )
   &SolveForcesAndVelocity(Clamp Velocity=false, Limit Acceleration=false)
   &DynamicMaterialParameters(
    Index 0 Param 1=&FloatFromCurve(FloatCurve=DataInterface() {propertyValues {Curve {keys [<at 0 1 RCIM_Cubic>, {arriveTangent=-2, interpMode=RCIM_Cubic, leaveTangent=-2, time=0.665, value=0.6902}, <at 1 -1 RCIM_Cubic>]}}})
    Index 0 Param 2=$Particles.MaterialRandom
    Index 0 Param 3=0.6
    Index 0 Param 4=1
    Module.Param0WriteEnabled=true
    Module.Param1WriteEnabled=true
    Module.Param2WriteEnabled=true
    Module.Param3WriteEnabled=true
   )
   &ScaleRibbonWidth(Ribbon Width Scale=&FloatFromCurve(
     FloatCurve=DataInterface() {propertyValues {Curve {keys [
         <at 0 1 RCIM_Cubic>
         {arriveTangent=-1.6026, interpMode=RCIM_Cubic, leaveTangent=-1.6026, tangentMode=RCTM_User, time=0.7967, value=0.5043}
         {arriveTangent=-2.0105, interpMode=RCIM_Cubic, leaveTangent=-2.0105, tangentMode=RCTM_User, time=1, value=0}
        ]}}}
     Scale Curve=0.18
    ))
  ]
  rendererClasses [&NiagaraRibbonRendererProperties]
  renderers [{rendererClass=&NiagaraRibbonRendererProperties}]
 }
 Trail_Blade {
  emitterUpdateScript [
   &EmitterState(Life Cycle Mode=::NewEnumerator1, Loop Behavior=::NewEnumerator1, Loop Duration=0.3, UseLoopDelay=false)
   &FN_SpawnAnimTrail_v2()
  ]
  particleSpawnScript [
   &FN_InitializeAnimTrailBoneIndecies(Second Bone or Socket Index=1)
   &FN_InitializeAnimTrail_v2(AnimTrailWidth=0.8, Lifetime=0.17)
   &self:ColorFromTOD_01(
    Afternoon:LinearColor=<rgba 1.44 1.52 1.6 2>
    Evening:LinearColor=<rgba 0.75 0.375 0.375 2>
    Morning:LinearColor=<rgba 0.75 0.1875 0.1875 2>
    Night:LinearColor=<rgba 0.7 1 0.875 2>
    TOD:Vector4f=&MakeVector4_FromLinearColor(LinearColor=$NPC.FortniteNPC.FortniteActiveTimeOfDay)
   ) {module=ColorFromTOD_01}
  ]
  particleUpdateScript [
   &FN_UpdateAnimTrail_v2()
   &SolveForcesAndVelocity(Clamp Velocity=false, Limit Acceleration=false)
   &DynamicMaterialParameters(
    Index 0 Param 1=&UniformRangedFloat(Minimum=0, Maximum=1)
    Index 0 Param 2=&UniformRangedFloat(Minimum=0, Maximum=1)
    Index 0 Param 3=1
    Index 0 Param 4=1
    Module.Param0WriteEnabled=true
    Module.Param1WriteEnabled=true
    Module.Param2WriteEnabled=true
    Module.Param3WriteEnabled=true
   )
   &ScaleRibbonWidth(Ribbon Width Scale=&FloatFromCurve(FloatCurve=DataInterface() {propertyValues {Curve {keys [
         {arriveTangent=-1, interpMode=RCIM_Cubic, time=0.8, value=1}
         {arriveTangent=-2.9906, interpMode=RCIM_Cubic, leaveTangent=-2.9906, tangentMode=RCTM_User, time=1, value=0}
        ]}}}))
   &ScaleColor(
    Scale Alpha=&FloatFromCurve(FloatCurve=DataInterface() {propertyValues {Curve {keys [
         {arriveTangent=-0.36, interpMode=RCIM_Cubic, leaveTangent=-0.36, tangentMode=RCTM_User, time=0, value=1}
         {arriveTangent=-1.7349, interpMode=RCIM_Cubic, leaveTangent=-1.7349, tangentMode=RCTM_User, time=1, value=0}
        ]}}})
    Scale RGB=&Multiply_Vector(
     A=<1 1 1>
     B=&MakeVectorFromLinearColorRGB(Color=&ColorFromCurve(ColorCurve=DataInterface() {propertyValues {
         AlphaCurve {keys [<at 0 1 RCIM_Cubic>]}
         BlueCurve {keys [<at 0 1 RCIM_Cubic>,<at 0.9913 1>]}
         GreenCurve {keys [<at 0 1 RCIM_Cubic>,<at 0.9913 1>]}
         RedCurve {keys [<at 0 1 RCIM_Cubic>,<at 0.9913 1>]}
        }}))
    )
    ScaleA=true
    ScaleRGB=true
   )
  ]
  rendererClasses [&NiagaraRibbonRendererProperties]
  renderers [{rendererClass=&NiagaraRibbonRendererProperties}]
 }
}
summary {
 systemName=NS_Pickaxe_Clash_V_Trail
 emitters [
  {emitterName=Tendril_Trail, rendererClasses [&NiagaraRibbonRendererProperties]}
  {emitterName=Tendril_Trail001, rendererClasses [&NiagaraRibbonRendererProperties]}
  {emitterName=Trail_Blade, rendererClasses [&NiagaraRibbonRendererProperties]}
  {emitterName=Tendril_Trail002, rendererClasses [&NiagaraRibbonRendererProperties]}
  {emitterName=Tendril_Trail003, rendererClasses [&NiagaraRibbonRendererProperties]}
 ]
 userVariables [
  User.RibbonIDOffset {type=NiagaraInt32}
  User.SkeletalMesh {
   type=NiagaraDataInterfaceSkeletalMesh
   defaultValue {
    struct=NiagaraExt_VariableValue_DataInterface
    value {dataInterface=&self:NiagaraDataInterfaceSkeletalMesh_1}
   }
  }
 ]
}
