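# Physics asset for the Clash V Male one-handed axe pickaxe cosmetic, giving its dangling tentacle bones bodies and joints for physics simulation.
# Six tentacle chains (A–F) of two to four small capsules each link bone to bone and attach at their first bone to one large lowerarm_r capsule.
# All swing and twist motions are Free, so the stored angle limits do not restrict rotation.
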PhysicsAsset /BRCosmetics/Weapons/FORT_Melee/Pickaxe_Clash_V_Male_1h/Meshes/Clash_V_Male_1h_Axe_Physics
bodies {
 TentaclesA_01 [{
   shapeType=Capsule
   length=3.0381
   radius=0.501
   shapeName=TentaclesA_01_capsule
   center=<-5.8137 0.0541 -0.3167>
   rotation=<rot 0 -88.8365 -87.1741>
  }]
 TentaclesA_02 [{
   shapeType=Capsule
   length=2.4091
   radius=0.5746
   shapeName=TentaclesA_02_capsule
   center=<-5.1304 0.0106 0.0256>
   rotation=<rot 0 94.9052 -82.4244>
  }]
 TentaclesA_03 [{
   shapeType=Capsule
   length=1.7689
   radius=0.5528
   shapeName=TentaclesA_03_capsule
   center=<-2.5106 0.3637 -0.9839>
   rotation=<rot 0 67.6691 -71.6184>
  }]
 TentaclesB_01 [{
   shapeType=Capsule
   length=2.876
   radius=1.4667
   shapeName=TentaclesB_01_capsule
   center=<-5.8395 -0.4836 -0.5141>
   rotation=<rot 0 -85.9799 -88.1674>
  }]
 TentaclesB_02 [{
   shapeType=Capsule
   length=1.8263
   radius=1.035
   shapeName=TentaclesB_02_capsule
   center=<-4.3863 0.2111 0.0595>
   rotation=<rot 0 95.535 -86.7839>
  }]
 TentaclesB_03 [{
   shapeType=Capsule
   length=2.3599
   radius=0.9045
   shapeName=TentaclesB_03_capsule
   center=<-3.6145 -0.001 -0.0278>
   rotation=<rot 0 91.3061 -87.8331>
  }]
 TentaclesB_04 [{
   shapeType=Capsule
   length=2.0698
   radius=0.6283
   shapeName=TentaclesB_04_capsule
   center=<-3.4759 -0.2194 -1.062>
   rotation=<rot 0 92.8451 -62.564>
  }]
 TentaclesC_01 [{
   shapeType=Capsule
   length=2.1849
   radius=0.9884
   shapeName=TentaclesC_01_capsule
   center=<-4.711 0.0992 0.3967>
   rotation=<rot 0 88.401 -88.7991>
  }]
 TentaclesC_02 [{
   shapeType=Capsule
   length=2.744
   radius=1.2195
   shapeName=TentaclesC_02_capsule
   center=<-3.3025 0.4196 -1.0046>
   rotation=<rot -2.9229 -90 39.0272>
  }]
 TentaclesD_01 [{
   shapeType=Capsule
   length=2.5733
   radius=1.0697
   shapeName=TentaclesD_01_capsule
   center=<-5.6652 -0.1459 0.0595>
   rotation=<rot 0 91.7607 -89.7284>
  }]
 TentaclesD_02 [{
   shapeType=Capsule
   length=3.5424
   radius=1.2438
   shapeName=TentaclesD_02_capsule
   center=<-4.3341 0.6752 -0.0803>
   rotation=<rot 0 -86.644 -80.591>
  }]
 TentaclesD_03 [{
   shapeType=Capsule
   length=3.8256
   radius=0.917
   shapeName=TentaclesD_03_capsule
   center=<-4.4728 -0.4242 1.2459>
   rotation=<rot 0 -87.5041 -79.4528>
  }]
 TentaclesD_04 [{
   shapeType=Capsule
   length=1.8809
   radius=1.1633
   shapeName=TentaclesD_04_capsule
   center=<-2.0758 -0.3489 2.1973>
   rotation=<rot 0 -109.4602 -88.7923>
  }]
 TentaclesE_01 [{
   shapeType=Capsule
   length=3.6602
   radius=0.6729
   shapeName=TentaclesE_01_capsule
   center=<-8.4949 -0.3685 0.0483>
   rotation=<rot 0 92.6216 -86.6595>
  }]
 TentaclesE_02 [{
   shapeType=Capsule
   length=2.8027
   radius=0.6789
   shapeName=TentaclesE_02_capsule
   center=<-4.9337 -0.7409 0.1814>
   rotation=<rot 0 75.7125 -85.2001>
  }]
 TentaclesE_03 [{
   shapeType=Capsule
   length=2.6235
   radius=1.0577
   shapeName=TentaclesE_03_capsule
   center=<-3.5018 -0.2954 0.2808>
   rotation=<rot 0 116.8661 -77.6768>
  }]
 TentaclesE_04 [{
   shapeType=Capsule
   length=1.1739
   radius=0.8294
   shapeName=TentaclesE_04_capsule
   center=<-2.1187 -0.714 0.451>
   rotation=<rot 0 -138.1052 -57.2102>
  }]
 TentaclesF_01 [{
   shapeType=Capsule
   length=4.5505
   radius=2.5202
   shapeName=TentaclesF_01_capsule
   center=<-6.2584 -0.4793 -0.4308>
   rotation=<rot 0 -113.8108 -89.9561>
  }]
 TentaclesF_02 [{
   shapeType=Capsule
   length=3.5133
   radius=1.1284
   shapeName=TentaclesF_02_capsule
   center=<-4.6946 -0.0235 0.3335>
   rotation=<rot 0 -89.6273 -88.6031>
  }]
 TentaclesF_03 [{
   shapeType=Capsule
   length=1.7415
   radius=0.296
   shapeName=TentaclesF_03_capsule
   center=<-3.9258 -0.9032 0.3324>
   rotation=<rot 0 -76.989 -82.499>
  }]
 lowerarm_r [{shapeType=Capsule, length=88.9518, radius=18.023, center=<-50.7761 5.176 -4.1534>, rotation=<rot 0 80.258 -85.1294>}]
}
constraints [
 {
  bone1Name=TentaclesD_01
  bone2Name=lowerarm_r
  swing1LimitDegrees=45
  swing1Motion=Free
  swing2LimitDegrees=45
  swing2Motion=Free
  twistLimitDegrees=45
  twistMotion=Free
 }
 {
  bone1Name=TentaclesD_02
  bone2Name=TentaclesD_01
  swing1LimitDegrees=45
  swing1Motion=Free
  swing2LimitDegrees=45
  swing2Motion=Free
  twistLimitDegrees=45
  twistMotion=Free
 }
 {
  bone1Name=TentaclesD_03
  bone2Name=TentaclesD_02
  swing1LimitDegrees=45
  swing1Motion=Free
  swing2LimitDegrees=45
  swing2Motion=Free
  twistLimitDegrees=45
  twistMotion=Free
 }
 {
  bone1Name=TentaclesD_04
  bone2Name=TentaclesD_03
  swing1LimitDegrees=45
  swing1Motion=Free
  swing2LimitDegrees=45
  swing2Motion=Free
  twistLimitDegrees=45
  twistMotion=Free
 }
 {
  bone1Name=TentaclesE_01
  bone2Name=lowerarm_r
  swing1LimitDegrees=45
  swing1Motion=Free
  swing2LimitDegrees=45
  swing2Motion=Free
  twistLimitDegrees=45
  twistMotion=Free
 }
 {
  bone1Name=TentaclesE_02
  bone2Name=TentaclesE_01
  swing1LimitDegrees=45
  swing1Motion=Free
  swing2LimitDegrees=45
  swing2Motion=Free
  twistLimitDegrees=45
  twistMotion=Free
 }
 {
  bone1Name=TentaclesE_03
  bone2Name=TentaclesE_02
  swing1LimitDegrees=45
  swing1Motion=Free
  swing2LimitDegrees=45
  swing2Motion=Free
  twistLimitDegrees=45
  twistMotion=Free
 }
 {
  bone1Name=TentaclesE_04
  bone2Name=TentaclesE_03
  swing1LimitDegrees=45
  swing1Motion=Free
  swing2LimitDegrees=45
  swing2Motion=Free
  twistLimitDegrees=45
  twistMotion=Free
 }
 {
  bone1Name=TentaclesF_01
  bone2Name=lowerarm_r
  swing1LimitDegrees=45
  swing1Motion=Free
  swing2LimitDegrees=45
  swing2Motion=Free
  twistLimitDegrees=45
  twistMotion=Free
 }
 {
  bone1Name=TentaclesF_02
  bone2Name=TentaclesF_01
  swing1LimitDegrees=45
  swing1Motion=Free
  swing2LimitDegrees=45
  swing2Motion=Free
  twistLimitDegrees=45
  twistMotion=Free
 }
 {
  bone1Name=TentaclesF_03
  bone2Name=TentaclesF_02
  swing1LimitDegrees=45
  swing1Motion=Free
  swing2LimitDegrees=45
  swing2Motion=Free
  twistLimitDegrees=45
  twistMotion=Free
 }
 {
  bone1Name=TentaclesC_01
  bone2Name=lowerarm_r
  swing1LimitDegrees=45
  swing1Motion=Free
  swing2LimitDegrees=45
  swing2Motion=Free
  twistLimitDegrees=45
  twistMotion=Free
 }
 {
  bone1Name=TentaclesC_02
  bone2Name=TentaclesC_01
  swing1LimitDegrees=45
  swing1Motion=Free
  swing2LimitDegrees=45
  swing2Motion=Free
  twistLimitDegrees=45
  twistMotion=Free
 }
 {
  bone1Name=TentaclesB_01
  bone2Name=lowerarm_r
  swing1LimitDegrees=45
  swing1Motion=Free
  swing2LimitDegrees=45
  swing2Motion=Free
  twistLimitDegrees=45
  twistMotion=Free
 }
 {
  bone1Name=TentaclesB_02
  bone2Name=TentaclesB_01
  swing1LimitDegrees=45
  swing1Motion=Free
  swing2LimitDegrees=45
  swing2Motion=Free
  twistLimitDegrees=45
  twistMotion=Free
 }
 {
  bone1Name=TentaclesB_03
  bone2Name=TentaclesB_02
  swing1LimitDegrees=45
  swing1Motion=Free
  swing2LimitDegrees=45
  swing2Motion=Free
  twistLimitDegrees=45
  twistMotion=Free
 }
 {
  bone1Name=TentaclesB_04
  bone2Name=TentaclesB_03
  swing1LimitDegrees=45
  swing1Motion=Free
  swing2LimitDegrees=45
  swing2Motion=Free
  twistLimitDegrees=45
  twistMotion=Free
 }
 {
  bone1Name=TentaclesA_01
  bone2Name=lowerarm_r
  swing1LimitDegrees=45
  swing1Motion=Free
  swing2LimitDegrees=45
  swing2Motion=Free
  twistLimitDegrees=45
  twistMotion=Free
 }
 {
  bone1Name=TentaclesA_02
  bone2Name=TentaclesA_01
  swing1LimitDegrees=45
  swing1Motion=Free
  swing2LimitDegrees=45
  swing2Motion=Free
  twistLimitDegrees=45
  twistMotion=Free
 }
 {
  bone1Name=TentaclesA_03
  bone2Name=TentaclesA_02
  swing1LimitDegrees=45
  swing1Motion=Free
  swing2LimitDegrees=45
  swing2Motion=Free
  twistLimitDegrees=45
  twistMotion=Free
 }
]
