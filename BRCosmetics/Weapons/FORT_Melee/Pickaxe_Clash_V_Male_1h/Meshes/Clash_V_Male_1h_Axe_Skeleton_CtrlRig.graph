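# Control Rig for the Clash_V_Male_1h pickaxe cosmetic that procedurally animates six tentacle bone chains (A–F) parented to lowerarm_r.
# On execution, six ChainHarmonics units run in sequence, one from each chain's root bone.
# All share identical wave-only settings (noise, sine ease-in, nearly flat curve), with Pendulum and Reach disabled.
# Debug drawing is enabled on every unit, and no variables or external inputs drive the motion.
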ControlRigBlueprint /BRCosmetics/Weapons/FORT_Melee/Pickaxe_Clash_V_Male_1h/Meshes/Clash_V_Male_1h_Axe_Skeleton_CtrlRig
graphs [
 RigVMFunctionLibrary {links [], nodes []}
 RigVMModel {nodes [
   n1 = &RigUnit_BeginExecution()
   /Script/ControlRig.RigUnit_ChainHarmonics(
    ChainRoot=TentaclesA_01
    DrawWorldOffset=(Rotation=(X=0,Y=0,Z=0,W=1),Translation=(X=0,Y=0,Z=0),Scale3D=(X=1,Y=1,Z=1))
    Pendulum=(bEnabled=False,PendulumStiffness=2,PendulumGravity=(X=0,Y=0,Z=0),PendulumBlend=0.75,PendulumDrag=0.98,PendulumMinimum=0,PendulumMaximum=0.1,PendulumEase=Linear,UnwindAxis=(X=0,Y=1,Z=0),UnwindMinimum=0.2,UnwindMaximum=0.05)
    Reach=(bEnabled=False,ReachTarget=(X=0,Y=0,Z=0),ReachAxis=(X=1,Y=0,Z=0),ReachMinimum=0,ReachMaximum=0,ReachEase=SineEaseInOut)
    Speed=(X=0.25,Y=0.25,Z=0.1)
    Wave=(bEnabled=True,WaveFrequency=(X=0.5,Y=0.5,Z=0.5),WaveAmplitude=(X=0.25,Y=1.5,Z=0.8),WaveOffset=(X=1,Y=0,Z=2),WaveNoise=(X=8,Y=8,Z=30),WaveMinimum=0,WaveMaximum=0.75,WaveEase=SineEaseIn)
    WaveCurve=(EditorCurveData=(Keys=((Time=-0.0198,Value=0.4929),(Time=0.9855,Value=0.5049)),PreInfinityExtrap=RCCE_Constant,PostInfinityExtrap=RCCE_Constant,DefaultValue=340282346638528859811704183484516925440))
    bDrawDebug=True
    ExecutePin=@n1.ExecutePin
   ) {is=struct RigVMUnitNode}
   > &RigUnit_ChainHarmonics(
    ChainRoot=TentaclesB_01
    DrawWorldOffset=(Rotation=(X=0,Y=0,Z=0,W=1),Translation=(X=0,Y=0,Z=0),Scale3D=(X=1,Y=1,Z=1))
    Pendulum=(bEnabled=False,PendulumStiffness=2,PendulumGravity=(X=0,Y=0,Z=0),PendulumBlend=0.75,PendulumDrag=0.98,PendulumMinimum=0,PendulumMaximum=0.1,PendulumEase=Linear,UnwindAxis=(X=0,Y=1,Z=0),UnwindMinimum=0.2,UnwindMaximum=0.05)
    Reach=(bEnabled=False,ReachTarget=(X=0,Y=0,Z=0),ReachAxis=(X=1,Y=0,Z=0),ReachMinimum=0,ReachMaximum=0,ReachEase=SineEaseInOut)
    Speed=(X=0.25,Y=0.25,Z=0.1)
    Wave=(bEnabled=True,WaveFrequency=(X=0.5,Y=0.5,Z=0.5),WaveAmplitude=(X=0.25,Y=1.5,Z=0.8),WaveOffset=(X=1,Y=0,Z=2),WaveNoise=(X=8,Y=8,Z=30),WaveMinimum=0,WaveMaximum=0.75,WaveEase=SineEaseIn)
    WaveCurve=(EditorCurveData=(Keys=((Time=-0.0198,Value=0.4929),(Time=0.9855,Value=0.5049)),PreInfinityExtrap=RCCE_Constant,PostInfinityExtrap=RCCE_Constant,DefaultValue=340282346638528859811704183484516925440))
    bDrawDebug=True
   )
   > &RigUnit_ChainHarmonics(
    ChainRoot=TentaclesC_01
    DrawWorldOffset=(Rotation=(X=0,Y=0,Z=0,W=1),Translation=(X=0,Y=0,Z=0),Scale3D=(X=1,Y=1,Z=1))
    Pendulum=(bEnabled=False,PendulumStiffness=2,PendulumGravity=(X=0,Y=0,Z=0),PendulumBlend=0.75,PendulumDrag=0.98,PendulumMinimum=0,PendulumMaximum=0.1,PendulumEase=Linear,UnwindAxis=(X=0,Y=1,Z=0),UnwindMinimum=0.2,UnwindMaximum=0.05)
    Reach=(bEnabled=False,ReachTarget=(X=0,Y=0,Z=0),ReachAxis=(X=1,Y=0,Z=0),ReachMinimum=0,ReachMaximum=0,ReachEase=SineEaseInOut)
    Speed=(X=0.25,Y=0.25,Z=0.1)
    Wave=(bEnabled=True,WaveFrequency=(X=0.5,Y=0.5,Z=0.5),WaveAmplitude=(X=0.25,Y=1.5,Z=0.8),WaveOffset=(X=1,Y=0,Z=2),WaveNoise=(X=8,Y=8,Z=30),WaveMinimum=0,WaveMaximum=0.75,WaveEase=SineEaseIn)
    WaveCurve=(EditorCurveData=(Keys=((Time=-0.0198,Value=0.4929),(Time=0.9855,Value=0.5049)),PreInfinityExtrap=RCCE_Constant,PostInfinityExtrap=RCCE_Constant,DefaultValue=340282346638528859811704183484516925440))
    bDrawDebug=True
   )
   > &RigUnit_ChainHarmonics(
    ChainRoot=TentaclesD_01
    DrawWorldOffset=(Rotation=(X=0,Y=0,Z=0,W=1),Translation=(X=0,Y=0,Z=0),Scale3D=(X=1,Y=1,Z=1))
    Pendulum=(bEnabled=False,PendulumStiffness=2,PendulumGravity=(X=0,Y=0,Z=0),PendulumBlend=0.75,PendulumDrag=0.98,PendulumMinimum=0,PendulumMaximum=0.1,PendulumEase=Linear,UnwindAxis=(X=0,Y=1,Z=0),UnwindMinimum=0.2,UnwindMaximum=0.05)
    Reach=(bEnabled=False,ReachTarget=(X=0,Y=0,Z=0),ReachAxis=(X=1,Y=0,Z=0),ReachMinimum=0,ReachMaximum=0,ReachEase=SineEaseInOut)
    Speed=(X=0.25,Y=0.25,Z=0.1)
    Wave=(bEnabled=True,WaveFrequency=(X=0.5,Y=0.5,Z=0.5),WaveAmplitude=(X=0.25,Y=1.5,Z=0.8),WaveOffset=(X=1,Y=0,Z=2),WaveNoise=(X=8,Y=8,Z=30),WaveMinimum=0,WaveMaximum=0.75,WaveEase=SineEaseIn)
    WaveCurve=(EditorCurveData=(Keys=((Time=-0.0198,Value=0.4929),(Time=0.9855,Value=0.5049)),PreInfinityExtrap=RCCE_Constant,PostInfinityExtrap=RCCE_Constant,DefaultValue=340282346638528859811704183484516925440))
    bDrawDebug=True
   )
   > &RigUnit_ChainHarmonics(
    ChainRoot=TentaclesE_01
    DrawWorldOffset=(Rotation=(X=0,Y=0,Z=0,W=1),Translation=(X=0,Y=0,Z=0),Scale3D=(X=1,Y=1,Z=1))
    Pendulum=(bEnabled=False,PendulumStiffness=2,PendulumGravity=(X=0,Y=0,Z=0),PendulumBlend=0.75,PendulumDrag=0.98,PendulumMinimum=0,PendulumMaximum=0.1,PendulumEase=Linear,UnwindAxis=(X=0,Y=1,Z=0),UnwindMinimum=0.2,UnwindMaximum=0.05)
    Reach=(bEnabled=False,ReachTarget=(X=0,Y=0,Z=0),ReachAxis=(X=1,Y=0,Z=0),ReachMinimum=0,ReachMaximum=0,ReachEase=SineEaseInOut)
    Speed=(X=0.25,Y=0.25,Z=0.1)
    Wave=(bEnabled=True,WaveFrequency=(X=0.5,Y=0.5,Z=0.5),WaveAmplitude=(X=0.25,Y=1.5,Z=0.8),WaveOffset=(X=1,Y=0,Z=2),WaveNoise=(X=8,Y=8,Z=30),WaveMinimum=0,WaveMaximum=0.75,WaveEase=SineEaseIn)
    WaveCurve=(EditorCurveData=(Keys=((Time=-0.0198,Value=0.4929),(Time=0.9855,Value=0.5049)),PreInfinityExtrap=RCCE_Constant,PostInfinityExtrap=RCCE_Constant,DefaultValue=340282346638528859811704183484516925440))
    bDrawDebug=True
   )
   > &RigUnit_ChainHarmonics(
    ChainRoot=TentaclesF_01
    DrawWorldOffset=(Rotation=(X=0,Y=0,Z=0,W=1),Translation=(X=0,Y=0,Z=0),Scale3D=(X=1,Y=1,Z=1))
    Pendulum=(bEnabled=False,PendulumStiffness=2,PendulumGravity=(X=0,Y=0,Z=0),PendulumBlend=0.75,PendulumDrag=0.98,PendulumMinimum=0,PendulumMaximum=0.1,PendulumEase=Linear,UnwindAxis=(X=0,Y=1,Z=0),UnwindMinimum=0.2,UnwindMaximum=0.05)
    Reach=(bEnabled=False,ReachTarget=(X=0,Y=0,Z=0),ReachAxis=(X=1,Y=0,Z=0),ReachMinimum=0,ReachMaximum=0,ReachEase=SineEaseInOut)
    Speed=(X=0.25,Y=0.25,Z=0.1)
    Wave=(bEnabled=True,WaveFrequency=(X=0.5,Y=0.5,Z=0.5),WaveAmplitude=(X=0.25,Y=1.5,Z=0.8),WaveOffset=(X=1,Y=0,Z=2),WaveNoise=(X=8,Y=8,Z=30),WaveMinimum=0,WaveMaximum=0.75,WaveEase=SineEaseIn)
    WaveCurve=(EditorCurveData=(Keys=((Time=-0.0198,Value=0.4929),(Time=0.9855,Value=0.5049)),PreInfinityExtrap=RCCE_Constant,PostInfinityExtrap=RCCE_Constant,DefaultValue=340282346638528859811704183484516925440))
    bDrawDebug=True
   )
   RigUnit_BeginExecution = &RigUnit_BeginExecution()
  ]}
]
hierarchy [
 lowerarm_r {loc=<0 0 0>, rot=<0.9528 0.0245 0.302 -0.0177>}
 lowerarm_twist_01_r {parent=lowerarm_r, loc=<-17.855 0 0>, rot=<0 0 0 1>}
 lowerarm_twist_02_r {parent=lowerarm_r, loc=<-9.6199 0 0>, rot=<0 0 0 1>}
 TentaclesD_01 {parent=lowerarm_r, loc=<-25.0657 6.7781 -2.5804>, rot=<0.3068 0.2272 0.2536 -0.8888>}
 TentaclesD_02 {parent=TentaclesD_01, loc=<-6.795 0.0368 0.0578>, rot=<0.033 -0.3195 -0.0303 -0.9465>}
 TentaclesD_03 {parent=TentaclesD_02, loc=<-6.8866 0 0>, rot=<0 0 -0.3821 -0.9241>}
 TentaclesD_04 {parent=TentaclesD_03, loc=<-7.6389 0 0>, rot=<0 0 0.282 -0.9594>}
 TentaclesE_01 {parent=lowerarm_r, loc=<-52.7333 9.7285 -5.1884>, rot=<0.4003 -0.1373 -0.0999 -0.9005>}
 TentaclesE_02 {parent=TentaclesE_01, loc=<-9.4404 0 0>, rot=<-0.0488 0.1451 0.2217 -0.963>}
 TentaclesE_03 {parent=TentaclesE_02, loc=<-6.8866 0 0>, rot=<-0.0066 -0.0559 0.1178 -0.9914>}
 TentaclesE_04 {parent=TentaclesE_03, loc=<-4.8868 0 0>, rot=<0 0 0 1>}
 TentaclesF_01 {parent=lowerarm_r, loc=<-67.7775 2.0517 -5.3494>, rot=<-0.4311 0.0372 0.1031 -0.8956>}
 TentaclesF_02 {parent=TentaclesF_01, loc=<-9.3135 -0.3049 -0.4784>, rot=<-0.0333 0.1806 0.0466 -0.9819>}
 TentaclesF_03 {parent=TentaclesF_02, loc=<-7.0392 0 0>, rot=<0 0 0 1>}
 TentaclesC_01 {parent=lowerarm_r, loc=<-35.3045 3.7878 2.8077>, rot=<0.3726 -0.011 -0.1756 -0.9112>}
 TentaclesC_02 {parent=TentaclesC_01, loc=<-5.8632 0.1633 0.2562>, rot=<-0.007 0.1036 -0.067 -0.9923>}
 TentaclesB_01 {parent=lowerarm_r, loc=<-18.372 4.0545 1.2205>, rot=<0.2955 -0.1047 0.204 -0.9274>}
 TentaclesB_02 {parent=TentaclesB_01, loc=<-6.79 0.0375 0.0589>, rot=<0.0179 0.1022 -0.3019 -0.9477>}
 TentaclesB_03 {parent=TentaclesB_02, loc=<-7.0392 0 0>, rot=<-0.0092 0.1212 -0.1115 -0.9863>}
 TentaclesB_04 {parent=TentaclesB_03, loc=<-5.4041 0 0>, rot=<0 0 0 1>}
 TentaclesA_01 {parent=lowerarm_r, loc=<-7.8043 -7.3745 -2.1537>, rot=<0.2763 -0.2244 -0.1005 -0.9291>}
 TentaclesA_02 {parent=TentaclesA_01, loc=<-7.0666 0 0>, rot=<-0.007 0.1036 -0.067 -0.9923>}
 TentaclesA_03 {parent=TentaclesA_02, loc=<-7.0392 0 0>, rot=<0 0 0 1>}
]
