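# Animation Blueprint for the Clash V Male one-handed axe pickaxe mesh, layering procedural rig and physics motion onto an input pose.
# The pose runs through a Control Rig, a Modify Bone node that replaces lowerarm_r rotation with zero, and a RigidBody simulation using the axe's physics asset.
# The same Control Rig class runs again after physics before output.
# The event graph holds only unconnected stub nodes, with no variables.
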AnimBlueprint /BRCosmetics/Weapons/FORT_Melee/Pickaxe_Clash_V_Male_1h/Meshes/Clash_V_Male_1h_Axe_AnimBP
target_skeleton=&/Clash_V_Male_1h_Axe_Skeleton
graphs [
 AnimGraph {nodes [
   AnimGraphNode_SubInput_0 = Input Pose()
   n1 = Control Rig(Source=@AnimGraphNode_SubInput_0) {
    alphaBoolBlend {blendOption=Linear}
    controlRigAssetReference {blueprintRigClass=&/Clash_V_Male_1h_Axe_Skeleton_CtrlRig.Clash_V_Male_1h_Axe_Skeleton_CtrlRig_C}
   }
   n2 = RigidBody(ComponentPose="Transform (Modify) Bone - Bone: lowerarm_r"(ComponentPose=Local To Component(LocalPose=@n1), Alpha=1, Translation="0,0,0", Rotation="0,0,0", Scale="1,1,1") {rotationMode=BMM_Replace, alphaBoolBlend {blendOption=Linear}, boneToModify {boneName=lowerarm_r}}) {
    overridePhysicsAsset=&/Clash_V_Male_1h_Axe_Physics
    alphaBoolBlend {blendOption=Linear}
   }
   n3 = Control Rig(Source=Component To Local(ComponentPose=@n2)) {
    alphaBoolBlend {blendOption=Linear}
    controlRigAssetReference {blueprintRigClass=&Clash_V_Male_1h_Axe_Skeleton_CtrlRig_C}
   }
   Output Pose(Result=@n3) {layerGroup=None}
  ]}
 EventGraph {nodes [
   TryGetPawnOwner()
   Event BlueprintUpdateAnimation()
  ]}
]
variables []
